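# Animation Blueprint for the CraneGame backpack, adding secondary motion to its plushies, crane and joystick.
# Spring bones jiggle plushie tips at tiny alphas, cut to zero above a pawn speed threshold.
# Chained AnimDynamics nodes sway plushie bases, crane and joystick within tight limits, mostly relative to the pack.
# The event graph halves crane track dynamics while skydiving and enables one panda head's interpolated offset and dynamics during non-skydiving jumps or falls.
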AnimBlueprint /BRCosmetics/Accessories/FORT_Backpacks/Backpack_F_MED_CraneGame/Meshes/F_MED_CraneGame_Pack_AnimBP
target_skeleton=&/F_MED_CraneGame_Pack_Skeleton
graphs [
 AnimGraph {nodes [
   n1 = Get Spring Alpha 02() -> (Spring Alpha 02:double)
   n2 = Spring controller - Bone: PandaHead_Tip_01_jnt!(ComponentPose:Component Space Pose Link Structure=Local To Component(LocalPose=Input Pose()), Alpha:float=@n1) -> (Pose:Component Space Pose Link Structure) {
    is=title AnimGraphNode_SpringBone
    springDamping=1.5
    springStiffness=150
    alphaBoolBlend {blendOption=Linear}
    springBone {boneName=PandaHead_Tip_01_jnt}
   }
   n3 = Get SpringAlpha() -> (SpringAlpha:double)
   n4 = Spring controller - Bone: Fishsticks_Tip_02_jnt!(
    ComponentPose:Component Space Pose Link Structure=Spring controller - Bone: Guff_Tip_jnt!(
     ComponentPose:Component Space Pose Link Structure=Spring controller - Bone: Peely_Tip_01_jnt!(ComponentPose:Component Space Pose Link Structure=@n2, Alpha:float=@n3) -> (Pose:Component Space Pose Link Structure) {
      is=title AnimGraphNode_SpringBone
      springDamping=2
      springStiffness=150
      alphaBoolBlend {blendOption=Linear}
      springBone {boneName=Peely_Tip_01_jnt}
     }
     Alpha:float=@n1
    ) -> (Pose:Component Space Pose Link Structure) {
     is=title AnimGraphNode_SpringBone
     springDamping=2
     springStiffness=200
     alphaBoolBlend {blendOption=Linear}
     springBone {boneName=Guff_Tip_jnt}
    }
    Alpha:float=@n3
   ) -> (Pose:Component Space Pose Link Structure) {
    is=title AnimGraphNode_SpringBone
    springDamping=1.5
    springStiffness=200
    alphaBoolBlend {blendOption=Linear}
    springBone {boneName=Fishsticks_Tip_02_jnt}
   }
   n5 = Spring controller - Bone: Fishsticks_Tip_01_jnt!(ComponentPose:Component Space Pose Link Structure=@n4, Alpha:float=@n3) -> (Pose:Component Space Pose Link Structure) {
    is=title AnimGraphNode_SpringBone
    springDamping=1.5
    springStiffness=150
    alphaBoolBlend {blendOption=Linear}
    springBone {boneName=Fishsticks_Tip_01_jnt}
   }
   n6 = Spring controller - Bone: PandaHead_Tip_02_jnt!(
    ComponentPose:Component Space Pose Link Structure=Spring controller - Bone: Peely_Tip_02_jnt!(ComponentPose:Component Space Pose Link Structure=@n5, Alpha:float=@n3) -> (Pose:Component Space Pose Link Structure) {
     is=title AnimGraphNode_SpringBone
     springDamping=2
     springStiffness=150
     alphaBoolBlend {blendOption=Linear}
     springBone {boneName=Peely_Tip_02_jnt}
    }
    Alpha:float=@n3
   ) -> (Pose:Component Space Pose Link Structure) {
    is=title AnimGraphNode_SpringBone
    springDamping=2
    springStiffness=200
    alphaBoolBlend {blendOption=Linear}
    springBone {boneName=PandaHead_Tip_02_jnt}
   }
   n7 = Spring controller - Bone: PandaHead_Tip_03_jnt!(
    ComponentPose:Component Space Pose Link Structure=Spring controller - Bone: Llama_Tip_jnt!(ComponentPose:Component Space Pose Link Structure=@n6, Alpha:float=@n1) -> (Pose:Component Space Pose Link Structure) {
     is=title AnimGraphNode_SpringBone
     springDamping=1.5
     springStiffness=200
     alphaBoolBlend {blendOption=Linear}
     springBone {boneName=Llama_Tip_jnt}
    }
    Alpha:float=@n3
   ) -> (Pose:Component Space Pose Link Structure) {
    is=title AnimGraphNode_SpringBone
    springDamping=1.5
    springStiffness=200
    alphaBoolBlend {blendOption=Linear}
    springBone {boneName=PandaHead_Tip_03_jnt}
   }
   n8 = Get Plushie Is Falling Alpha() -> (Plushie Is Falling Alpha:double)
   n9 = "Transform (Modify) Bone - Bone: PandaHead_Base_02_jnt"(ComponentPose=@n7, Alpha=@n8, Translation="0,-0.5,0.25", Rotation="0,30,0", Scale="1,1,1") {
    rotationMode=BMM_Replace
    rotationSpace=BCS_BoneSpace
    translationMode=BMM_Replace
    translationSpace=BCS_BoneSpace
    alphaBoolBlend {blendOption=Linear}
    alphaScaleBiasClamp {bInterpResult=true, interpSpeedDecreasing=5, interpSpeedIncreasing=2}
    boneToModify {boneName=PandaHead_Base_02_jnt}
    rotation {yaw=30}
    translation {y=-0.5, z=0.25}
   }
   n10 = Anim Dynamics - Bone: PandaHead_Base_02_jnt(ComponentPose=@n9, Alpha=@n8, ExternalForce="0,0,-980") {
    angularDampingOverride=0.9
    angularSpringConstant=1600
    bAngularSpring=true
    bLinearSpring=true
    bOverrideAngularDamping=true
    bOverrideLinearDamping=true
    gravityScale=0
    linearDampingOverride=0.995
    linearSpringConstant=50
    simulationSpace=RootRelative
    alphaBoolBlend {blendOption=Linear}
    alphaScaleBiasClamp {bInterpResult=true, interpSpeedDecreasing=5, interpSpeedIncreasing=2}
    boundBone {boneName=PandaHead_Base_02_jnt}
    componentAppliedLinearAccClamp=<6000 6000 6000>
    componentLinearAccScale=<0.3 0.3 0.3>
    componentLinearVelScale=<0.7 0.7 0.7>
    gravityOverride {z=-980}
    physicsBodyDefinitions [{
      sphereCollisionRadius=0
      boundBone {boneName=PandaHead_Base_02_jnt}
      boxExtents=<7 7 7>
      constraintSetup {
       angularLimitsMax=<10 10 10>
       angularLimitsMin=<-10 -10 -10>
       angularTarget=<1 0 0>
       linearAxesMax=<0.1 0.5 0.25>
       linearAxesMin=<-0.25 -1 -0.25>
      }
      localJointOffset=<4 0 0>
     }]
    relativeSpaceBone {boneName=pack}
   }
   n11 = Anim Dynamics - Bone: PandaHead_Base_01_jnt(ComponentPose=@n10, Alpha=0.75, ExternalForce="0,0,-980") {
    angularDampingOverride=0.9
    angularSpringConstant=1600
    bAngularSpring=true
    bLinearSpring=true
    bOverrideAngularDamping=true
    bOverrideLinearDamping=true
    gravityScale=0
    linearDampingOverride=0.995
    linearSpringConstant=50
    simulationSpace=RootRelative
    alphaBoolBlend {blendOption=Linear}
    alphaScaleBiasClamp {interpSpeedDecreasing=5, interpSpeedIncreasing=2}
    boundBone {boneName=PandaHead_Base_01_jnt}
    componentAppliedLinearAccClamp=<6000 6000 6000>
    componentLinearAccScale=<0.3 0.3 0.3>
    componentLinearVelScale=<0.7 0.7 0.7>
    gravityOverride {z=-980}
    physicsBodyDefinitions [{
      sphereCollisionRadius=0
      boundBone {boneName=PandaHead_Base_01_jnt}
      boxExtents=<7 7 7>
      constraintSetup {
       angularTargetAxis=AxisZ
       twistAxis=AxisZ
       angularLimitsMax=<10 0 10>
       angularLimitsMin=<-10 0 -10>
       angularTarget=<0 0 1>
       linearAxesMax=<0.25 0.25 0.25>
       linearAxesMin=<-1 -0.25 -0.25>
      }
      localJointOffset=<4 0 0>
     }]
    relativeSpaceBone {boneName=pack}
   }
   n12 = Anim Dynamics - Bone: Peely_Base_01_jnt(ComponentPose=@n11, Alpha=0.5, ExternalForce="0,0,-980") {
    angularDampingOverride=0.9
    angularSpringConstant=1600
    bAngularSpring=true
    bLinearSpring=true
    bOverrideAngularDamping=true
    bOverrideLinearDamping=true
    gravityScale=0
    linearDampingOverride=0.995
    linearSpringConstant=50
    simulationSpace=RootRelative
    alphaBoolBlend {blendOption=Linear}
    alphaScaleBiasClamp {interpSpeedDecreasing=5, interpSpeedIncreasing=2}
    boundBone {boneName=Peely_Base_01_jnt}
    componentAppliedLinearAccClamp=<6000 6000 6000>
    componentLinearAccScale=<0.3 0.3 0.3>
    componentLinearVelScale=<0.7 0.7 0.7>
    gravityOverride {z=-980}
    physicsBodyDefinitions [{
      sphereCollisionRadius=0
      boundBone {boneName=Peely_Base_01_jnt}
      boxExtents=<7 7 7>
      constraintSetup {
       angularTargetAxis=AxisZ
       twistAxis=AxisZ
       angularLimitsMax=<10 10 10>
       angularLimitsMin=<-10 -10 -10>
       angularTarget=<0 0 1>
       linearAxesMax=<0.25 0.25 0.25>
       linearAxesMin=<-1 -0.25 -0.25>
      }
      localJointOffset=<4 0 0>
     }]
    relativeSpaceBone {boneName=pack}
   }
   n13 = Anim Dynamics - Bone: Guff_Base_jnt(ComponentPose=@n12, Alpha=1, ExternalForce="0,0,-980") {
    angularDampingOverride=0.9
    angularSpringConstant=1600
    bAngularSpring=true
    bLinearSpring=true
    bOverrideAngularDamping=true
    bOverrideLinearDamping=true
    gravityScale=0
    linearDampingOverride=0.995
    linearSpringConstant=50
    simulationSpace=RootRelative
    alphaBoolBlend {blendOption=Linear}
    alphaScaleBiasClamp {interpSpeedDecreasing=5, interpSpeedIncreasing=2}
    boundBone {boneName=Guff_Base_jnt}
    componentAppliedLinearAccClamp=<6000 6000 6000>
    componentLinearAccScale=<0.3 0.3 0.3>
    componentLinearVelScale=<0.7 0.7 0.7>
    gravityOverride {z=-980}
    physicsBodyDefinitions [{
      sphereCollisionRadius=0
      boundBone {boneName=Guff_Base_jnt}
      boxExtents=<7 7 7>
      constraintSetup {
       angularTargetAxis=AxisZ
       twistAxis=AxisZ
       angularLimitsMax=<0 0 0>
       angularLimitsMin=<0 0 0>
       angularTarget=<0 0 1>
       linearAxesMax=<0.25 0.15 0.15>
       linearAxesMin=<-0.5 -0.15 -0.15>
      }
      localJointOffset=<4 0 0>
     }]
    relativeSpaceBone {boneName=pack}
   }
   n14 = Anim Dynamics - Bone: Fishsticks_Base_02_jnt(ComponentPose=@n13, Alpha=0.5, ExternalForce="0,0,-980") {
    angularDampingOverride=0.9
    angularSpringConstant=1600
    bAngularSpring=true
    bLinearSpring=true
    bOverrideAngularDamping=true
    bOverrideLinearDamping=true
    gravityScale=0
    linearDampingOverride=0.995
    linearSpringConstant=50
    simulationSpace=RootRelative
    alphaBoolBlend {blendOption=Linear}
    alphaScaleBiasClamp {interpSpeedDecreasing=5, interpSpeedIncreasing=2}
    boundBone {boneName=Fishsticks_Base_02_jnt}
    componentAppliedLinearAccClamp=<6000 6000 6000>
    componentLinearAccScale=<0.3 0.3 0.3>
    componentLinearVelScale=<0.7 0.7 0.7>
    gravityOverride {z=-980}
    physicsBodyDefinitions [{
      sphereCollisionRadius=0
      boundBone {boneName=Fishsticks_Base_02_jnt}
      boxExtents=<7 7 7>
      constraintSetup {
       angularTargetAxis=AxisZ
       twistAxis=AxisZ
       angularLimitsMax=<10 10 10>
       angularLimitsMin=<-10 -10 -10>
       angularTarget=<0 0 1>
       linearAxesMax=<0.25 0.25 0.25>
       linearAxesMin=<-1 -0.25 -0.25>
      }
      localJointOffset=<4 0 0>
     }]
    relativeSpaceBone {boneName=pack}
   }
   n15 = Anim Dynamics - Bone: Fishsticks_Base_01_jnt(ComponentPose=@n14, Alpha=1, ExternalForce="0,0,-980") {
    angularDampingOverride=0.9
    angularSpringConstant=1600
    bAngularSpring=true
    bLinearSpring=true
    bOverrideAngularDamping=true
    bOverrideLinearDamping=true
    gravityScale=0
    linearDampingOverride=0.995
    linearSpringConstant=50
    simulationSpace=RootRelative
    alphaBoolBlend {blendOption=Linear}
    alphaScaleBiasClamp {interpSpeedDecreasing=5, interpSpeedIncreasing=2}
    boundBone {boneName=Fishsticks_Base_01_jnt}
    componentAppliedLinearAccClamp=<6000 6000 6000>
    componentLinearAccScale=<0.3 0.3 0.3>
    componentLinearVelScale=<0.7 0.7 0.7>
    gravityOverride {z=-980}
    physicsBodyDefinitions [{
      sphereCollisionRadius=0
      boundBone {boneName=Fishsticks_Base_01_jnt}
      boxExtents=<7 7 7>
      constraintSetup {
       angularTargetAxis=AxisZ
       twistAxis=AxisZ
       angularLimitsMax=<10 10 10>
       angularLimitsMin=<-10 -10 -10>
       angularTarget=<0 0 1>
       linearAxesMax=<0.25 0.25 0.25>
       linearAxesMin=<-1 -0.25 -0.25>
      }
      localJointOffset=<4 0 0>
     }]
    relativeSpaceBone {boneName=pack}
   }
   n16 = Anim Dynamics - Bone: Peely_Base_02_jnt(ComponentPose=@n15, Alpha=0.5, ExternalForce="0,0,-980") {
    angularDampingOverride=0.9
    angularSpringConstant=1600
    bAngularSpring=true
    bLinearSpring=true
    bOverrideAngularDamping=true
    bOverrideLinearDamping=true
    gravityScale=0
    linearDampingOverride=0.995
    linearSpringConstant=50
    simulationSpace=RootRelative
    alphaBoolBlend {blendOption=Linear}
    alphaScaleBiasClamp {interpSpeedDecreasing=5, interpSpeedIncreasing=2}
    boundBone {boneName=Peely_Base_02_jnt}
    componentAppliedLinearAccClamp=<6000 6000 6000>
    componentLinearAccScale=<0.3 0.3 0.3>
    componentLinearVelScale=<0.7 0.7 0.7>
    gravityOverride {z=-980}
    physicsBodyDefinitions [{
      sphereCollisionRadius=0
      boundBone {boneName=Peely_Base_02_jnt}
      boxExtents=<7 7 7>
      constraintSetup {
       angularTargetAxis=AxisZ
       twistAxis=AxisZ
       angularLimitsMax=<5 0 0>
       angularLimitsMin=<-5 0 0>
       angularTarget=<0 0 1>
       linearAxesMax=<0.25 0 0.25>
       linearAxesMin=<0 -0.25 0>
      }
      localJointOffset=<4 0 0>
     }]
    relativeSpaceBone {boneName=pack}
   }
   n17 = Anim Dynamics - Bone: Llama_Base_jnt(ComponentPose=@n16, Alpha=1, ExternalForce="0,0,-980") {
    angularDampingOverride=0.9
    angularSpringConstant=1600
    bAngularSpring=true
    bLinearSpring=true
    bOverrideAngularDamping=true
    bOverrideLinearDamping=true
    gravityScale=0
    linearDampingOverride=0.995
    linearSpringConstant=50
    simulationSpace=RootRelative
    alphaBoolBlend {blendOption=Linear}
    alphaScaleBiasClamp {interpSpeedDecreasing=5, interpSpeedIncreasing=2}
    boundBone {boneName=Llama_Base_jnt}
    componentAppliedLinearAccClamp=<6000 6000 6000>
    componentLinearAccScale=<0.3 0.3 0.3>
    componentLinearVelScale=<0.7 0.7 0.7>
    gravityOverride {z=-980}
    physicsBodyDefinitions [{
      sphereCollisionRadius=0
      boundBone {boneName=Llama_Base_jnt}
      boxExtents=<7 7 7>
      constraintSetup {
       angularTargetAxis=AxisZ
       twistAxis=AxisZ
       angularLimitsMax=<0 0 0>
       angularLimitsMin=<0 0 0>
       angularTarget=<0 0 1>
       linearAxesMax=<0.25 0.25 0.25>
      }
      localJointOffset=<4 0 0>
     }]
    relativeSpaceBone {boneName=pack}
   }
   n18 = Anim Dynamics - Bone: CraneTrack_jnt(ComponentPose=@n17, Alpha=Get Is SkyDiving Alpha() -> (Is SkyDiving Alpha:double), ExternalForce="0,0,0") {
    angularDampingOverride=0.9
    angularSpringConstant=200000
    bAngularSpring=true
    bLinearSpring=true
    bOverrideAngularDamping=true
    bOverrideLinearDamping=true
    gravityScale=0
    linearDampingOverride=0.995
    linearSpringConstant=50
    simulationSpace=BoneRelative
    alphaBoolBlend {blendOption=Linear}
    boundBone {boneName=CraneTrack_jnt}
    componentAppliedLinearAccClamp=<6000 6000 6000>
    componentLinearAccScale=<0.3 0.3 0.3>
    componentLinearVelScale=<0.7 0.7 0.7>
    gravityOverride {z=-980}
    physicsBodyDefinitions [{
      sphereCollisionRadius=0
      boundBone {boneName=CraneTrack_jnt}
      boxExtents=<7 7 7>
      constraintSetup {
       angularTargetAxis=AxisZ
       twistAxis=AxisZ
       angularLimitsMax=<0 0 0>
       angularLimitsMin=<0 0 0>
       angularTarget=<0 0 1>
       linearAxesMax=<0 1 0>
       linearAxesMin=<0 -5 0>
      }
      localJointOffset=<-0.1 -0.2 0>
     }]
    relativeSpaceBone {boneName=CraneGame_Jnt}
   }
   n19 = Anim Dynamics - Bone: Crane_Twist_jnt(ComponentPose=@n18, Alpha=1, ExternalForce="0,0,-980") {
    angularDampingOverride=0.9
    angularSpringConstant=1600
    bAngularSpring=true
    bLinearSpring=true
    bOverrideAngularDamping=true
    bOverrideLinearDamping=true
    gravityScale=0
    linearDampingOverride=0.995
    linearSpringConstant=50
    simulationSpace=RootRelative
    alphaBoolBlend {blendOption=Linear}
    boundBone {boneName=Crane_Twist_jnt}
    componentAppliedLinearAccClamp=<6000 6000 6000>
    componentLinearAccScale=<0.3 0.3 0.3>
    componentLinearVelScale=<0.7 0.7 0.7>
    gravityOverride {z=-980}
    physicsBodyDefinitions [{
      sphereCollisionRadius=0
      boundBone {boneName=Crane_Twist_jnt}
      boxExtents=<7 7 7>
      constraintSetup {angularLimitsMax=<0 0 15>, angularLimitsMin=<0 0 -15>, angularTarget=<1 0 0>}
      localJointOffset=<4 0 0>
     }]
    relativeSpaceBone {boneName=pack}
   }
   n20 = Anim Dynamics - Bone: JoyStick_Base_jnt(ComponentPose=@n19, Alpha=1, ExternalForce="0,0,-980") {
    angularDampingOverride=0.9
    angularSpringConstant=1600
    bAngularSpring=true
    bLinearSpring=true
    bOverrideAngularDamping=true
    bOverrideLinearDamping=true
    gravityScale=0
    linearDampingOverride=0.995
    linearSpringConstant=50
    simulationSpace=RootRelative
    alphaBoolBlend {blendOption=Linear}
    boundBone {boneName=JoyStick_Base_jnt}
    componentAppliedLinearAccClamp=<6000 6000 6000>
    componentLinearAccScale=<0.3 0.3 0.3>
    componentLinearVelScale=<0.7 0.7 0.7>
    gravityOverride {z=-980}
    physicsBodyDefinitions [{
      sphereCollisionRadius=0
      boundBone {boneName=JoyStick_Base_jnt}
      boxExtents=<7 7 7>
      constraintSetup {angularLimitsMax=<0 15 15>, angularLimitsMin=<0 -15 -15>, angularTarget=<1 0 0>}
      localJointOffset=<4 0 0>
     }]
    relativeSpaceBone {boneName=pack}
   }
   Output Pose(Result=Component To Local(ComponentPose=@n20)) {layerGroup=None}
  ]}
 EventGraph {nodes [
   Event BlueprintUpdateAnimation()
   > n1 = Sequence!(execute) -> (then_0, then_1, then_2:Exec)
   > n2 = Set pawnVelocitySpeed(pawnVelocitySpeed:double=Vector Length(A=Get PawnVelocity())) -> (Output_Get:double)
   > n3 = Set pawnVelocitySpeedCutoff(pawnVelocitySpeedCutoff:double=SelectFloat(A=0, B=1, bPickA="float > float"(A=@n2, B=1200) {is=title K2Node_CallFunction})) -> (Output_Get:double)
   > n4 = Set SpringAlpha(SpringAlpha:double=float * float(A=0.0025, B=@n3) {is=title K2Node_PromotableOperator}) -> (Output_Get:double)
   > Set Spring Alpha 02(Spring Alpha 02:double=float * float(A=0.5, B=@n4)) -> (Output_Get:double)
   n5 = Branch(execute=@n1.then_1, Condition=Get bIsSkyDiving())
   > Set Is SkyDiving Alpha(Is SkyDiving Alpha:double=0.5) -> (Output_Get:double)
   Set Is SkyDiving Alpha(execute=@n5.else, Is SkyDiving Alpha=1)
   n6 = Get bIsSkyDiving()
   n7 = Get bIsFalling()
   n8 = Get bIsJumping()
   n9 = Branch(execute=@n1.then_2, Condition=AND Boolean(A=OR Boolean(A=@n7, B=@n8), B=NOT Boolean(A=@n6)))
   > Set Plushie Is Falling Alpha(Plushie Is Falling Alpha:double=1) -> (Output_Get:double)
   Set Plushie Is Falling Alpha(execute=@n9.else, Plushie Is Falling Alpha=0)
   TryGetPawnOwner()
  ]}
]
variables [
 Is SkyDiving Alpha {type=number}
 pawnVelocitySpeed {type=number}
 pawnVelocitySpeedCutoff {type=number}
 Plushie Is Falling Alpha {type=number}
 Spring Alpha 02 {type=number}
 SpringAlpha {type=number}
]
